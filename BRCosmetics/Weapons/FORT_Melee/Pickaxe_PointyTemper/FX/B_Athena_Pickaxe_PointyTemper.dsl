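; Blueprint for the PointyTemper pickaxe cosmetic that feeds swing montage information to its swing particle effect.
; On each left or right swing it runs the parent behavior, then gets a section index via SetupSwingMontageSectionIndex from the instigator.
; That index is written to the swing VFX component as the SwingSectionIndex integer parameter, so the effect can differ per montage section.
; BeginPlay and Tick are empty, and the construction script only calls the parent.
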(event EventBeginPlay)

(event EventTick
  (params (DeltaSeconds :type Float%20%28single-precision%29)))

(event EventSwingLeft
  ;[0 0 0] Swing
  (|CallParentFunction)
  (bind _section_index (Utilities|SetupSwingMontageSectionIndex (Variables|Actor|GetInstigator))
    (:Out
      (Effects|Components|ParticleSystem|SetIntParameter (CallFunction|GetSwingVFXComponent) "SwingSectionIndex" _section_index))))

(event EventSwingRight
  ;[0 0 0] Swing
  (|CallParentFunction)
  (bind _section_index (Utilities|SetupSwingMontageSectionIndex (Variables|Actor|GetInstigator))
    (:Out
      (Effects|Components|ParticleSystem|SetIntParameter (CallFunction|GetSwingVFXComponent) "SwingSectionIndex" _section_index))))

(fn ConstructionScript
  (|CallParentFunction))
