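# Niagara ribbon trail effect for the PointyTemper pickaxe, using anim-trail modules to follow bones or sockets on a skeletal mesh.
# Claws_Trails spawns five short-lived dark red trails, while Hand_Green_Trail and Red_BG_Trail layer bright green and longer-lasting red ribbons on a shared socket pair.
# Curves shape color, fade and width, modulated by system TrailFade and color multiplier, time-of-day or front-end scalars, plus a material parameter on the green trail.
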NiagaraSystem /BRCosmetics/Weapons/FORT_Melee/Pickaxe_PointyTemper/FX/NS_Pickaxe_PointyTemper_Trail
dependencies {
 usedDataInterfaces [&NiagaraDataInterfaceCurve, &NiagaraDataInterfaceColorCurve, &NiagaraDataInterfaceSkeletalMesh]
 usedDynamicInputs [&self:SectionIndexFloatSelector_01]
 usedModules [
  &self:SystemSpawnScript.SystemScriptSource.SystemScriptGraph.NiagaraNodeAssignment_0.SetVariables_#1
  &SystemState
  &self:SystemSpawnScript.SystemScriptSource.SystemScriptGraph.NiagaraNodeAssignment_1.SetVariables_#2
  &self:SystemSpawnScript.SystemScriptSource.SystemScriptGraph.NiagaraNodeAssignment_3.SetVariables_#3
 ]
}
emitters {
 Claws_Trails {
  emitterUpdateScript [
   &EmitterState(Life Cycle Mode=::NewEnumerator1, Loop Behavior=::NewEnumerator1, Loop Duration=0.25, UseLoopDelay=false)
   &FN_SpawnAnimTrail(Number of Anim Trails=5)
  ]
  particleSpawnScript [
   &FN_InitializeAnimTrailBoneIndecies(First Bone or Socket Index=1)
   &FN_InitializeAnimTrailBoneIndecies(AnimTrail Selection=1, First Bone or Socket Index=2, Second Bone or Socket Index=2) {module=FN_InitializeAnimTrailBoneIndecies001}
   &FN_InitializeAnimTrailBoneIndecies(AnimTrail Selection=2, First Bone or Socket Index=3, Second Bone or Socket Index=3) {module=FN_InitializeAnimTrailBoneIndecies002}
   &FN_InitializeAnimTrailBoneIndecies(AnimTrail Selection=3, First Bone or Socket Index=4, Second Bone or Socket Index=4) {module=FN_InitializeAnimTrailBoneIndecies003}
   &FN_InitializeAnimTrailBoneIndecies(AnimTrail Selection=4, First Bone or Socket Index=5, Second Bone or Socket Index=5) {module=FN_InitializeAnimTrailBoneIndecies004}
   &FN_InitializeAnimTrail(AnimTrailWidth=10, Lifetime=0.13, SkeletalMesh=$System.Skeletal Mesh)
  ]
  particleUpdateScript [
   &FN_UpdateAnimTrail(SkeletalMesh=$System.Skeletal Mesh)
   &ParticleState()
   &ScaleRibbonWidth(Ribbon Width Scale=&FloatFromCurve(FloatCurve=DataInterface() {propertyValues {Curve {keys [
         {arriveTangent=-0.0514, interpMode=RCIM_Cubic, leaveTangent=-0.0514, tangentMode=RCTM_User, time=0, value=1}
         {arriveTangent=-1.0387, interpMode=RCIM_Cubic, leaveTangent=-1.0387, tangentMode=RCTM_User, time=1, value=0.3}
        ]}}}))
   &ScaleColor(
    Scale Alpha=&Multiply_Float(
     A=&Multiply_Float(
      A=&FloatFromCurve(
       FloatCurve=DataInterface() {propertyValues {Curve {keys [
           {interpMode=RCIM_Cubic, leaveTangent=-1.5196, tangentMode=RCTM_Break, time=0, value=1}
           {arriveTangent=-1.059, interpMode=RCIM_Cubic, leaveTangent=-1.059, tangentMode=RCTM_User, time=0.4, value=0.4}
           {arriveTangent=-0.0239, interpMode=RCIM_Cubic, tangentMode=RCTM_Break, time=1, value=0}
          ]}}}
       Scale Curve=&Multiply_Float(A=3, B=$System.TrailFade)
      )
      B=&Multiply_Float(A=1, B=&FNTimeOfDayScalar(Night Scale=0.7, Morning Scale=0.8, Evening Scale=0.8))
     )
     B=&MakeFloatFromLinearColor(LinearColor=$System.TODColorMultiplier, Channel:ENiagaraLinearColor_Channels=::NewEnumerator3)
    )
    Scale RGB=&Multiply_Vector(
     A=&MakeVectorFromLinearColorRGB(Color=&ColorFromCurve(ColorCurve=DataInterface() {propertyValues {
         AlphaCurve {keys []}
         BlueCurve {keys [<at 0 0>,<at 0.4611 0>]}
         GreenCurve {keys [<at 0 0>,<at 0.4611 0>]}
         RedCurve {keys [<at 0 0.0677>,<at 0.4611 0>]}
        }}))
     B=&MakeVectorFromLinearColorRGB(Color=$System.TODColorMultiplier)
    )
    ScaleA=true
    ScaleRGB=true
   )
  ]
  rendererClasses [&NiagaraRibbonRendererProperties]
  renderers [{rendererClass=&NiagaraRibbonRendererProperties}]
 }
 Hand_Green_Trail {
  emitterUpdateScript [
   &EmitterState(Life Cycle Mode=::NewEnumerator1, Loop Behavior=::NewEnumerator1, Loop Duration=0.3, UseLoopDelay=false)
   &FN_SpawnAnimTrail(Spawn Interval=0)
  ]
  particleSpawnScript [
   &FN_InitializeAnimTrailBoneIndecies(First Bone or Socket Index=6, Second Bone or Socket Index=7)
   &FN_InitializeAnimTrail(AnimTrailWidth=2, Color=<rgba 1.5 1.5 1.5 1>, Lifetime=0.55, SkeletalMesh=$System.Skeletal Mesh)
  ]
  particleUpdateScript [
   &FN_UpdateAnimTrail(SkeletalMesh=$System.Skeletal Mesh)
   &ParticleState()
   &ScaleColor(
    Scale Alpha=&Multiply_Float(
     A=&Multiply_Float(
      A=&FloatFromCurve(
       FloatCurve=DataInterface() {propertyValues {Curve {keys [
           {arriveTangent=-0.0116, interpMode=RCIM_Cubic, leaveTangent=-2.1412, tangentMode=RCTM_Break, time=0, value=1}
           {arriveTangent=-1.5974, interpMode=RCIM_Cubic, leaveTangent=-1.5974, tangentMode=RCTM_User, time=0.1, value=0.8}
           {arriveTangent=-0.0239, interpMode=RCIM_Cubic, tangentMode=RCTM_Break, time=1, value=0}
          ]}}}
       Scale Curve=&Multiply_Float(A=$System.TrailFade, B=1)
      )
      B=&Multiply_Float(A=1, B=&FNTimeOfDayScalar(Night Scale=0.6, Morning Scale=0.8, Evening Scale=0.8))
     )
     B=&MakeFloatFromLinearColor(LinearColor=$System.TODColorMultiplier, Channel:ENiagaraLinearColor_Channels=::NewEnumerator3)
    )
    Scale RGB=&Multiply_Vector(
     A=&Multiply_VectorByFloat(
      Vector=&Multiply_VectorByFloat(
       Vector=&MakeVectorFromLinearColorRGB(Color=&ColorFromCurve(ColorCurve=DataInterface() {propertyValues {
           AlphaCurve {keys []}
           BlueCurve {keys [<at 0.0424 0.1091 RCIM_Cubic>,<at 0.7693 0.0024 RCIM_Cubic>]}
           GreenCurve {keys [
             {arriveTangent=-0.9565, interpMode=RCIM_Cubic, leaveTangent=-0.9565, tangentMode=RCTM_User, time=0.0424, value=0.5}
             <at 0.7693 0.01 RCIM_Cubic>
            ]}
           RedCurve {keys [
             {arriveTangent=0.0615, interpMode=RCIM_Cubic, leaveTangent=0.0615, tangentMode=RCTM_User, time=0.0424, value=0.2896}
             <at 0.7693 0.0054 RCIM_Cubic>
            ]}
          }}))
       Float=&FloatFromCurve(FloatCurve=DataInterface() {propertyValues {Curve {keys [{arriveTangent=-21.6817, interpMode=RCIM_Cubic, leaveTangent=-21.6817, tangentMode=RCTM_User, time=0, value=5}, <at 0.2 1>, <at 1 1>]}}})
      )
      Float=&FNTimeOfDayScalar(Night Scale=2, Morning Scale=1, Evening Scale=1)
     )
     B=&MakeVectorFromLinearColorRGB(Color=$System.TODColorMultiplier)
    )
    ScaleA=true
    ScaleRGB=true
   )
   &DynamicMaterialParameters(
    Index 0 Param 1=&FloatFromCurve(FloatCurve=DataInterface() {propertyValues {Curve {keys [{arriveTangent=-2.3944, interpMode=RCIM_Cubic, leaveTangent=-2.3944, tangentMode=RCTM_User, time=0.4, value=1}, <at 1 -1 RCIM_Cubic>]}}})
    Module.Index0Param0Write=true
    Module.Index0Param1Write=true
    Module.Index0Param2Write=true
    Module.Index0Param3Write=true
   )
   &ScaleRibbonWidth(Ribbon Width Scale=&FloatFromCurve(FloatCurve=DataInterface() {propertyValues {Curve {keys [{arriveTangent=-3.2419, interpMode=RCIM_Cubic, leaveTangent=-3.2419, tangentMode=RCTM_User, time=0, value=1.25}, <at 0.2 1>]}}}))
  ]
  rendererClasses [&NiagaraRibbonRendererProperties]
  renderers [{rendererClass=&NiagaraRibbonRendererProperties}]
 }
 Red_BG_Trail {
  emitterUpdateScript [
   &EmitterState(Life Cycle Mode=::NewEnumerator1, Loop Behavior=::NewEnumerator1, Loop Duration=0.3, UseLoopDelay=false)
   &FN_SpawnAnimTrail(AnimTrailWidthMode=::FromSecond, Spawn Interval=0)
  ]
  particleSpawnScript [
   &FN_InitializeAnimTrailBoneIndecies(First Bone or Socket Index=6, Second Bone or Socket Index=7)
   &FN_InitializeAnimTrail(AnimTrailWidth=2, Color=<rgba 0.7 0.7 0.7 0.8>, Lifetime=0.68, SkeletalMesh=$System.Skeletal Mesh)
  ]
  particleUpdateScript [
   &FN_UpdateAnimTrail(SkeletalMesh=$System.Skeletal Mesh)
   &ParticleState()
   &ScaleColor(
    Scale Alpha=&Multiply_Float(
     A=&Multiply_Float(
      A=&FloatFromCurve(FloatCurve=DataInterface() {propertyValues {Curve {keys [
           {arriveTangent=-0.0116, interpMode=RCIM_Cubic, leaveTangent=-3.1782, tangentMode=RCTM_Break, time=0.3, value=1}
           {arriveTangent=-0.029, interpMode=RCIM_Cubic, tangentMode=RCTM_Break, time=1, value=0}
          ]}}})
      B=&Multiply_Float(A=$System.TrailFade, B=&FN_FrontEndFloatModifier(Gameplay=1.25, Front End=0.5))
     )
     B=&MakeFloatFromLinearColor(LinearColor=$System.TODColorMultiplier, Channel:ENiagaraLinearColor_Channels=::NewEnumerator3)
    )
    Scale RGB=&Multiply_Vector(
     A=&MakeVectorFromLinearColorRGB(Color=&ColorFromCurve(ColorCurve=DataInterface() {propertyValues {
         AlphaCurve {keys []}
         BlueCurve {keys [<at 0.2551 0.0131 RCIM_Cubic>,<at 0.7882 0 RCIM_Cubic>]}
         GreenCurve {keys [
           {arriveTangent=-0.9565, interpMode=RCIM_Cubic, leaveTangent=-0.9565, tangentMode=RCTM_User, time=0.2551, value=0.0131}
           <at 0.7882 0 RCIM_Cubic>
          ]}
         RedCurve {keys [
           {arriveTangent=0.0615, interpMode=RCIM_Cubic, leaveTangent=0.0615, tangentMode=RCTM_User, time=0.2551, value=0.4271}
           <at 0.7882 0.3177 RCIM_Cubic>
          ]}
        }}))
     B=&MakeVectorFromLinearColorRGB(Color=$System.TODColorMultiplier)
    )
    ScaleA=true
    ScaleRGB=true
   )
  ]
  rendererClasses [&NiagaraRibbonRendererProperties]
  renderers [{rendererClass=&NiagaraRibbonRendererProperties}]
 }
}
summary {
 systemName=NS_Pickaxe_PointyTemper_Trail
 emitters [
  {emitterName=Claws_Trails, rendererClasses [&NiagaraRibbonRendererProperties]}
  {emitterName=Hand_Green_Trail, rendererClasses [&NiagaraRibbonRendererProperties]}
  {emitterName=Red_BG_Trail, rendererClasses [&NiagaraRibbonRendererProperties]}
 ]
 userVariables [
  User.bIsFrontEnd {type=NiagaraBool}
  User.bIsFrontEndPreview {type=NiagaraBool}
  User.RibbonIDOffset {type=NiagaraInt32}
  User.SwingSectionIndex {type=NiagaraInt32}
 ]
}
