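# Animation Blueprint for the PointyTemper cosmetic pickaxe that plays its idle animation and holds that animation's first frame during melee swings.
# Each update, the owning player's MeleeAnim curve sets Melee Anim Playing, which switches the Idle/Swing state machine.
# The additive bone-rotation chain for the hand and fingers does not reach Output Pose.
# The phase-offset sine oscillators sit on an unconnected Sequence, and nothing reads their variables.
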AnimBlueprint /BRCosmetics/Weapons/FORT_Melee/Pickaxe_PointyTemper/Meshes/PointyTemper_Axe_AnimBP
target_skeleton=&/PointyTemper_Axe_Skeleton
graphs [
 AnimGraph {nodes [
   AnimGraphNode_SubInput_0 = Input Pose()
   n1 = "Transform (Modify) Bone - Bone: dyn_thumb_01"(ComponentPose=Local To Component(LocalPose=@AnimGraphNode_SubInput_0), Alpha=1, Translation="0,0,0", Rotation="-10,0,5", Scale="1,1,1") {
    rotationMode=BMM_Additive
    rotationSpace=BCS_BoneSpace
    scaleSpace=BCS_WorldSpace
    translationMode=BMM_Additive
    translationSpace=BCS_BoneSpace
    alphaBoolBlend {blendOption=Linear}
    boneToModify {boneName=dyn_thumb_01}
    rotation {pitch=-10, roll=5}
   }
   n2 = "Transform (Modify) Bone - Bone: dyn_thumb_02"(ComponentPose=@n1, Alpha=1, Translation="0,0,0", Rotation="-15,0,0", Scale="1,1,1") {
    rotationMode=BMM_Additive
    translationMode=BMM_Additive
    translationSpace=BCS_BoneSpace
    alphaBoolBlend {blendOption=Linear}
    boneToModify {boneName=dyn_thumb_02}
    rotation {pitch=-15}
   }
   n3 = "Transform (Modify) Bone - Bone: dyn_thumb_03"(ComponentPose=@n2, Alpha=1, Translation="0,0,0", Rotation="-5,0,0", Scale="1,1,1") {
    rotationMode=BMM_Additive
    rotationSpace=BCS_BoneSpace
    translationMode=BMM_Additive
    translationSpace=BCS_BoneSpace
    alphaBoolBlend {blendOption=Linear}
    boneToModify {boneName=dyn_thumb_03}
    rotation {pitch=-5}
   }
   n4 = "Transform (Modify) Bone - Bone: dyn_hand"(ComponentPose=@n3, Alpha=1, Translation="0,0,0", Rotation="-10,0,0", Scale="1,1,1") {
    rotationMode=BMM_Additive
    rotationSpace=BCS_BoneSpace
    translationMode=BMM_Additive
    translationSpace=BCS_BoneSpace
    alphaBoolBlend {blendOption=Linear}
    boneToModify {boneName=dyn_hand}
    rotation {pitch=-10}
   }
   n5 = "Transform (Modify) Bone - Bone: dyn_thumb_01"(ComponentPose=@n4, Alpha=1, Translation="0,0,0", Rotation="6,-4,-4", Scale="1,1,1") {
    rotationMode=BMM_Additive
    rotationSpace=BCS_BoneSpace
    translationMode=BMM_Additive
    translationSpace=BCS_BoneSpace
    alphaBoolBlend {blendOption=Linear}
    boneToModify {boneName=dyn_thumb_01}
    rotation=<rot 6 -4 -4>
   }
   n6 = "Transform (Modify) Bone - Bone: dyn_thumb_02"(ComponentPose=@n5, Alpha=1, Translation="0,0,0", Rotation="5,0,0", Scale="1,1,1") {
    rotationMode=BMM_Additive
    rotationSpace=BCS_BoneSpace
    translationMode=BMM_Additive
    translationSpace=BCS_BoneSpace
    alphaBoolBlend {blendOption=Linear}
    boneToModify {boneName=dyn_thumb_02}
    rotation {pitch=5}
   }
   n7 = "Transform (Modify) Bone - Bone: dyn_thumb_03"(ComponentPose=@n6, Alpha=1, Translation="0,0,0", Rotation="7,0,0", Scale="1,1,1") {
    rotationMode=BMM_Additive
    rotationSpace=BCS_BoneSpace
    translationMode=BMM_Additive
    translationSpace=BCS_BoneSpace
    alphaBoolBlend {blendOption=Linear}
    boneToModify {boneName=dyn_thumb_03}
    rotation {pitch=7}
   }
   n8 = "Transform (Modify) Bone - Bone: dyn_index_01"(ComponentPose=@n7, Alpha=1, Translation="0,0,0", Rotation="12,0,0", Scale="1,1,1") {
    rotationMode=BMM_Additive
    rotationSpace=BCS_BoneSpace
    translationMode=BMM_Additive
    translationSpace=BCS_BoneSpace
    alphaBoolBlend {blendOption=Linear}
    boneToModify {boneName=dyn_index_01}
    rotation {pitch=12}
   }
   n9 = "Transform (Modify) Bone - Bone: dyn_index_02"(ComponentPose=@n8, Alpha=1, Translation="0,0,0", Rotation="10,0,0", Scale="1,1,1") {
    rotationMode=BMM_Additive
    rotationSpace=BCS_BoneSpace
    translationMode=BMM_Additive
    translationSpace=BCS_BoneSpace
    alphaBoolBlend {blendOption=Linear}
    boneToModify {boneName=dyn_index_02}
    rotation {pitch=10}
   }
   n10 = "Transform (Modify) Bone - Bone: dyn_index_03"(ComponentPose=@n9, Alpha=1, Translation="0,0,0", Rotation="7,0,0", Scale="1,1,1") {
    rotationMode=BMM_Additive
    rotationSpace=BCS_BoneSpace
    translationMode=BMM_Additive
    translationSpace=BCS_BoneSpace
    alphaBoolBlend {blendOption=Linear}
    boneToModify {boneName=dyn_index_03}
    rotation {pitch=7}
   }
   n11 = "Transform (Modify) Bone - Bone: dyn_middle_01"(ComponentPose=@n10, Alpha=1, Translation="0,0,0", Rotation="12,0,0", Scale="1,1,1") {
    rotationMode=BMM_Additive
    rotationSpace=BCS_BoneSpace
    translationMode=BMM_Additive
    translationSpace=BCS_BoneSpace
    alphaBoolBlend {blendOption=Linear}
    boneToModify {boneName=dyn_middle_01}
    rotation {pitch=12}
   }
   n12 = "Transform (Modify) Bone - Bone: dyn_middle_02"(ComponentPose=@n11, Alpha=1, Translation="0,0,0", Rotation="10,0,0", Scale="1,1,1") {
    rotationMode=BMM_Additive
    rotationSpace=BCS_BoneSpace
    translationMode=BMM_Additive
    translationSpace=BCS_BoneSpace
    alphaBoolBlend {blendOption=Linear}
    boneToModify {boneName=dyn_middle_02}
    rotation {pitch=10}
   }
   n13 = "Transform (Modify) Bone - Bone: dyn_middle_03"(ComponentPose=@n12, Alpha=1, Translation="0,0,0", Rotation="7,0,0", Scale="1,1,1") {
    rotationMode=BMM_Additive
    rotationSpace=BCS_BoneSpace
    translationMode=BMM_Additive
    translationSpace=BCS_BoneSpace
    alphaBoolBlend {blendOption=Linear}
    boneToModify {boneName=dyn_middle_03}
    rotation {pitch=7}
   }
   n14 = "Transform (Modify) Bone - Bone: dyn_ring_01"(ComponentPose=@n13, Alpha=1, Translation="0,0,0", Rotation="10,0,0", Scale="1,1,1") {
    rotationMode=BMM_Additive
    rotationSpace=BCS_BoneSpace
    translationMode=BMM_Additive
    translationSpace=BCS_BoneSpace
    alphaBoolBlend {blendOption=Linear}
    boneToModify {boneName=dyn_ring_01}
    rotation {pitch=10}
   }
   n15 = "Transform (Modify) Bone - Bone: dyn_ring_02"(ComponentPose=@n14, Alpha=1, Translation="0,0,0", Rotation="8,0,0", Scale="1,1,1") {
    rotationMode=BMM_Additive
    rotationSpace=BCS_BoneSpace
    translationMode=BMM_Additive
    translationSpace=BCS_BoneSpace
    alphaBoolBlend {blendOption=Linear}
    boneToModify {boneName=dyn_ring_02}
    rotation {pitch=8}
   }
   n16 = "Transform (Modify) Bone - Bone: dyn_ring_03"(ComponentPose=@n15, Alpha=1, Translation="0,0,0", Rotation="6,0,0", Scale="1,1,1") {
    rotationMode=BMM_Additive
    rotationSpace=BCS_BoneSpace
    translationMode=BMM_Additive
    translationSpace=BCS_BoneSpace
    alphaBoolBlend {blendOption=Linear}
    boneToModify {boneName=dyn_ring_03}
    rotation {pitch=6}
   }
   n17 = "Transform (Modify) Bone - Bone: dyn_pinky_01"(ComponentPose=@n16, Alpha=1, Translation="0,0,0", Rotation="8,0,0", Scale="1,1,1") {
    rotationMode=BMM_Additive
    rotationSpace=BCS_BoneSpace
    translationMode=BMM_Additive
    translationSpace=BCS_BoneSpace
    alphaBoolBlend {blendOption=Linear}
    boneToModify {boneName=dyn_pinky_01}
    rotation {pitch=8}
   }
   n18 = "Transform (Modify) Bone - Bone: dyn_pinky_02"(ComponentPose=@n17, Alpha=1, Translation="0,0,0", Rotation="4,0,0", Scale="1,1,1") {
    rotationMode=BMM_Additive
    rotationSpace=BCS_BoneSpace
    translationMode=BMM_Additive
    translationSpace=BCS_BoneSpace
    alphaBoolBlend {blendOption=Linear}
    boneToModify {boneName=dyn_pinky_02}
    rotation {pitch=4}
   }
   n19 = "Transform (Modify) Bone - Bone: dyn_pinky_03"(ComponentPose=@n18, Alpha=1, Translation="0,0,0", Rotation="4,0,0", Scale="1,1,1") {
    rotationMode=BMM_Additive
    rotationSpace=BCS_BoneSpace
    translationMode=BMM_Additive
    translationSpace=BCS_BoneSpace
    alphaBoolBlend {blendOption=Linear}
    boneToModify {boneName=dyn_pinky_03}
    rotation {pitch=4}
   }
   n20 = "Transform (Modify) Bone - Bone: dyn_hand"(ComponentPose=@n19, Alpha=1, Translation="0,0,0", Rotation="10,0,0", Scale="1,1,1") {
    rotationMode=BMM_Additive
    rotationSpace=BCS_BoneSpace
    translationMode=BMM_Additive
    translationSpace=BCS_BoneSpace
    alphaBoolBlend {blendOption=Linear}
    boneToModify {boneName=dyn_hand}
    rotation {pitch=10}
   }
   Component To Local(ComponentPose=@n20)
   AnimGraphNode_StateMachine_0 = PointyTemper State Machine() {is=title AnimGraphNode_StateMachine}
   Output Pose(Result=@AnimGraphNode_StateMachine_0) {layerGroup=None}
  ]}
 PointyTemper State Machine {
  entry=Idle
  path=AnimGraph.AnimGraphNode_StateMachine_0.PointyTemper State Machine
  ids=id
  nodes [
   AnimStateNode_0 = Swing() {is=title AnimStateNode}
   AnimStateNode_2 = Idle()
   AnimStateTransitionNode_1 = AnimStateTransitionNode() {crossfadeDuration=0.1}
   AnimStateTransitionNode_2 = AnimStateTransitionNode() {crossfadeDuration=0}
  ]
  links [
   {
    dst=Swing
    rule=AnimGraph.AnimGraphNode_StateMachine_0.PointyTemper State Machine.AnimStateTransitionNode_1.Transition
    src=Idle
    via=AnimStateTransitionNode_1
   }
   {
    dst=Idle
    rule=AnimGraph.AnimGraphNode_StateMachine_0.PointyTemper State Machine.AnimStateTransitionNode_2.Transition
    src=Swing
    via=AnimStateTransitionNode_2
   }
  ]
 }
 Swing {
  path=AnimGraph.AnimGraphNode_StateMachine_0.PointyTemper State Machine.AnimStateNode_0.Swing
  nodes [Output Animation Pose(Result="Sequence Evaluator (by time) 'PointyTemper_Idle_Anim_WEP'"!(ExplicitTime:float=0) -> (Pose:Pose Link Structure) {is=title AnimGraphNode_SequenceEvaluator, sequence=/BRCosmetics/Weapons/FORT_Melee/Pickaxe_PointyTemper/Anim/PointyTemper_Idle_Anim_WEP})]
 }
 Idle {
  path=AnimGraph.AnimGraphNode_StateMachine_0.PointyTemper State Machine.AnimStateNode_2.Idle
  nodes [Output Animation Pose(Result=Sequence Player 'PointyTemper_Idle_Anim_WEP'() {sequence=&PointyTemper_Idle_Anim_WEP})]
 }
 Transition {
  path=AnimGraph.AnimGraphNode_StateMachine_0.PointyTemper State Machine.AnimStateTransitionNode_1.Transition
  nodes [Result(bCanEnterTransition=Get Melee Anim Playing() -> (Melee Anim Playing:bool))]
 }
 Transition {
  path=AnimGraph.AnimGraphNode_StateMachine_0.PointyTemper State Machine.AnimStateTransitionNode_2.Transition
  nodes [Result(bCanEnterTransition=NOT Boolean(A=Get Melee Anim Playing()))]
 }
 EventGraph {nodes [
   Event BlueprintUpdateAnimation()
   > n1 = Cast To FortWeaponPickaxeAthena!(execute, Object=GetOwningActor()) -> (then, CastFailed, AsFort Weapon Pickaxe Athena:Fort Weapon Pickaxe Athena Object Reference) {is=title K2Node_DynamicCast}
   n2 = GetOwner(self:Actor Object Reference=@n1)
   > Is Valid(InputObject=@n2)
   n3 = GetComponentByClass!(self:Actor Object Reference=@n2, ComponentClass:Actor Component Class Reference=/Script/Engine.SkeletalMeshComponent) -> (ReturnValue:Skeletal Mesh Component Object Reference)
   > Is Valid(InputObject=@n3)
   > n4 = Cast To FortPlayerAnimInstance(Object=GetAnimInstance(self=@n3))
   > Set Melee Anim Playing(Melee Anim Playing:bool="float >= float"(A=GetCurveValue(self=@n4, CurveName=MeleeAnim), B=1) {is=title K2Node_CallFunction}) -> (Output_Get:bool)
   n5 = Sequence!(execute) -> (then_0, then_1, then_2:Exec, then_3:Exec, then_4:Exec, then_5:Exec)
   n6 = GetRealTimeSeconds()
   n7 = float * float(A=@n6, B=2) {is=title K2Node_PromotableOperator}
   n8 = "Sin (Radians)"(A=float + float(A=@n7, B=4.4))
   > Set Rotate_Thumb_01(Rotate_Thumb_01:double=float * float(A=float + float(A=@n8, B=1), B=0.5)) -> (Output_Get:double)
   n9 = "Sin (Radians)"(A=float + float(A=@n7, B=3.8))
   > Set Rotate_Thumb_02(Rotate_Thumb_02:double=float * float(A=float + float(A=@n9, B=1), B=0.5)) -> (Output_Get:double)
   n10 = "Sin (Radians)"(A=float + float(A=@n7, B=3.2))
   > Set Rotate_Thumb_03(Rotate_Thumb_03:double=float * float(A=float + float(A=@n10, B=1), B=0.5)) -> (Output_Get:double)
   n11 = float * float(A=@n6, B=2)
   n12 = "Sin (Radians)"(A=float + float(A=@n11, B=4.6))
   Set Rotate_Index_01(execute=@n5.then_1, Rotate_Index_01:double=float * float(A=float + float(A=@n12, B=1), B=0.5)) -> (Output_Get:double)
   n13 = "Sin (Radians)"(A=float + float(A=@n11, B=3.6))
   > Set Rotate_Index_02(Rotate_Index_02:double=float * float(A=float + float(A=@n13, B=1), B=0.5)) -> (Output_Get:double)
   n14 = "Sin (Radians)"(A=float + float(A=@n11, B=3))
   > Set Rotate_Index_03(Rotate_Index_03:double=float * float(A=float + float(A=@n14, B=1), B=0.5)) -> (Output_Get:double)
   n15 = float * float(A=@n6, B=2)
   n16 = "Sin (Radians)"(A=float + float(A=@n15, B=5))
   Set Rotate_Middle_01(execute=@n5.then_2, Rotate_Middle_01:double=float * float(A=float + float(A=@n16, B=1), B=0.5)) -> (Output_Get:double)
   n17 = "Sin (Radians)"(A=float + float(A=@n15, B=4))
   > Set Rotate_Middle_02(Rotate_Middle_02:double=float * float(A=float + float(A=@n17, B=1), B=0.5)) -> (Output_Get:double)
   n18 = "Sin (Radians)"(A=float + float(A=@n15, B=3.4))
   > Set Rotate_Middle_03(Rotate_Middle_03:double=float * float(A=float + float(A=@n18, B=1), B=0.5)) -> (Output_Get:double)
   n19 = float * float(A=@n6, B=2)
   n20 = "Sin (Radians)"(A=float + float(A=@n19, B=5.4))
   Set Rotate_Ring_01(execute=@n5.then_3, Rotate_Ring_01:double=float * float(A=float + float(A=@n20, B=1), B=0.5)) -> (Output_Get:double)
   n21 = "Sin (Radians)"(A=float + float(A=@n19, B=4.4))
   > Set Rotate_Ring_02(Rotate_Ring_02:double=float * float(A=float + float(A=@n21, B=1), B=0.5)) -> (Output_Get:double)
   n22 = "Sin (Radians)"(A=float + float(A=@n19, B=3.8))
   > Set Rotate_Ring_03(Rotate_Ring_03:double=float * float(A=float + float(A=@n22, B=1), B=0.5)) -> (Output_Get:double)
   n23 = float * float(A=@n6, B=2)
   n24 = "Sin (Radians)"(A=float + float(A=@n23, B=5.8))
   Set Rotate_Pinky_01(execute=@n5.then_4, Rotate_Pinky_01:double=float * float(A=float + float(A=@n24, B=1), B=0.5)) -> (Output_Get:double)
   n25 = "Sin (Radians)"(A=float + float(A=@n23, B=4.8))
   > Set Rotate_Pinky_02(Rotate_Pinky_02:double=float * float(A=float + float(A=@n25, B=1), B=0.5)) -> (Output_Get:double)
   n26 = "Sin (Radians)"(A=float + float(A=@n23, B=4.2))
   > Set Rotate_Pinky_03(Rotate_Pinky_03:double=float * float(A=float + float(A=@n26, B=1), B=0.5)) -> (Output_Get:double)
   n27 = "Sin (Radians)"(A=float + float(A=float * float(A=@n6, B=2), B=5.8))
   Set Rotate_Hand(execute=@n5.then_5, Rotate_Hand:double=float * float(A=float + float(A=@n27, B=1), B=0.5)) -> (Output_Get:double)
   TryGetPawnOwner()
   MapRangeClamped(Value=0, InRangeA=-5, InRangeB=-20)
  ]}
]
variables [
 Melee Anim Playing {type=boolean}
 Rotate_Hand {type=number}
 Rotate_Index_01 {type=number}
 Rotate_Index_02 {type=number}
 Rotate_Index_03 {type=number}
 Rotate_Middle_01 {type=number}
 Rotate_Middle_02 {type=number}
 Rotate_Middle_03 {type=number}
 Rotate_Pinky_01 {type=number}
 Rotate_Pinky_02 {type=number}
 Rotate_Pinky_03 {type=number}
 Rotate_Ring_01 {type=number}
 Rotate_Ring_02 {type=number}
 Rotate_Ring_03 {type=number}
 Rotate_Thumb_01 {type=number}
 Rotate_Thumb_02 {type=number}
 Rotate_Thumb_03 {type=number}
]
