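# Physics asset for the PointyTemper two-handed pickaxe mesh, a cosmetic melee weapon, giving it a simplified physics shape.
# A single body on the Melee_TwoHanded_Handle bone uses one capsule, offset along the bone's Z axis and rotated from the bone's orientation.
# It has no constraints, so the asset defines one body with no jointed parts.
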PhysicsAsset /BRCosmetics/Weapons/FORT_Melee/Pickaxe_PointyTemper/Meshes/PointyTemper_Axe_PhysicsAsset
bodies {Melee_TwoHanded_Handle [{shapeType=Capsule, length=51.5013, radius=12.5075, center=<-0.3118 0.3942 27.1903>, rotation=<rot 0.2427 -90 1.9084>}]}
constraints []
